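# Physics asset for the BrineGleam pickaxe on a pawn, set up for a dangling bag with bodies for the character's legs.
# A three-sphere bag chain, growing larger toward the end, hangs from a tiny master capsule through swing-limited joints with very tight twist.
# Thigh and calf capsules on both legs give the bag body shapes to collide against, each calf joined to its thigh with free angular motion.
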PhysicsAsset /BRCosmetics/Weapons/FORT_Melee/Pickaxe_BrineGleam/Meshes/BrineGleam_Axe_PhysicsFE_OnPawn
bodies {
 calf_l [{shapeType=Capsule, length=30, radius=8, shapeName=thigh_l_capsule, center=<-26 0 -4.2974>, rotation=<rot 90 180 0>}]
 calf_r [{shapeType=Capsule, length=30, radius=9.5, shapeName=thigh_r_capsule, center=<26 0 -1.3537>, rotation=<rot 90 0 0>}]
 dyn_Bag_01_jnt [{shapeType=Sphere, radius=4.5, shapeName=dyn_Bag_01_jnt_sphere, center=<3.1116 0 0>}]
 dyn_Bag_02_jnt [{shapeType=Sphere, radius=7.5, shapeName=dyn_Bag_02_jnt_sphere, center=<1.63 0 0>}]
 dyn_Bag_03_jnt [{shapeType=Sphere, radius=13.5, shapeName=dyn_Bag_03_jnt_sphere, center=<6.2249 1.2041 0>}]
 dyn_Bag_MSTR [{shapeType=Capsule, length=1, radius=1, shapeName=dyn_Bag_MSTR_capsule, rotation=<rot 0 -90 0>}]
 thigh_l [{shapeType=Capsule, length=34.6616, radius=9.5479, shapeName=thigh_l_capsule, center=<-22.1907 -0.6066 0.3009>, rotation=<rot 90 180 0>}]
 thigh_r [{shapeType=Capsule, length=34.6616, radius=9.5479, shapeName=thigh_r_capsule, center=<22.1907 0.6066 -0.3009>, rotation=<rot 90 0 0>}]
}
constraints [
 {bone1Name=dyn_Bag_02_jnt, bone2Name=dyn_Bag_01_jnt, swing1LimitDegrees=45, swing2LimitDegrees=90, twistLimitDegrees=2}
 {bone1Name=dyn_Bag_03_jnt, bone2Name=dyn_Bag_02_jnt, swing1LimitDegrees=45, swing2LimitDegrees=90, twistLimitDegrees=2}
 {
  bone1Name=calf_r
  bone2Name=thigh_r
  swing1LimitDegrees=45
  swing1Motion=Free
  swing2LimitDegrees=45
  swing2Motion=Free
  twistLimitDegrees=45
  twistMotion=Free
 }
 {
  bone1Name=calf_l
  bone2Name=thigh_l
  swing1LimitDegrees=45
  swing1Motion=Free
  swing2LimitDegrees=45
  swing2Motion=Free
  twistLimitDegrees=45
  twistMotion=Free
 }
 {bone1Name=dyn_Bag_01_jnt, bone2Name=dyn_Bag_MSTR, swing1LimitDegrees=45, swing2LimitDegrees=90, twistLimitDegrees=2}
]
