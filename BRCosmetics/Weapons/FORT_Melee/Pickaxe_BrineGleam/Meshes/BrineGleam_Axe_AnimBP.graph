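# Animation Blueprint for the BrineGleam pickaxe, adding physics motion and pose offsets in gameplay and front end.
# Each update caches the parent body's root, pelvis and leg transforms and selects simulation settings and gravity by context.
# The AnimGraph chains world-space replacements of those body bones, additive handle, root and bag offsets, then three RigidBody nodes with default, off-pawn and on-pawn physics assets.
# Customization loading sets gender, body-size and front-end rotation flags.
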AnimBlueprint /BRCosmetics/Weapons/FORT_Melee/Pickaxe_BrineGleam/Meshes/BrineGleam_Axe_AnimBP
target_skeleton=&/BrineGleam_Axe_Skeleton
graphs [
 AnimGraph {nodes [
   AnimGraphNode_SubInput_0 = Input Pose()
   n1 = "Transform (Modify) Bone - Bone: body_root"(ComponentPose=Local To Component(LocalPose=@AnimGraphNode_SubInput_0), Alpha=1, Translation="0,0,0", Rotation="0,0,0", Scale="1,1,1") {
    rotationMode=BMM_Replace
    rotationSpace=BCS_WorldSpace
    scaleSpace=BCS_BoneSpace
    translationMode=BMM_Replace
    translationSpace=BCS_WorldSpace
    alphaBoolBlend {blendOption=Linear}
    alphaScaleBiasClamp {bInterpResult=true, interpSpeedDecreasing=1, interpSpeedIncreasing=1, inRange {max=0, min=-40}, outRange {max=0, min=1}}
    boneToModify {boneName=body_root}
   }
   n2 = "Transform (Modify) Bone - Bone: pelvis"(ComponentPose=@n1, Alpha=1, Translation="0,0,0", Rotation="0,0,0", Scale="1,1,1") {
    rotationMode=BMM_Replace
    rotationSpace=BCS_WorldSpace
    scaleSpace=BCS_BoneSpace
    translationMode=BMM_Replace
    translationSpace=BCS_WorldSpace
    alphaBoolBlend {blendOption=Linear}
    alphaScaleBiasClamp {inRange {max=0, min=-40}, outRange {max=0, min=1}}
    boneToModify {boneName=pelvis}
   }
   n3 = "Transform (Modify) Bone - Bone: thigh_l"(ComponentPose=@n2, Alpha=1, Translation="0,0,0", Rotation="0,0,0", Scale="1,1,1") {
    rotationMode=BMM_Replace
    rotationSpace=BCS_WorldSpace
    scaleSpace=BCS_BoneSpace
    translationMode=BMM_Replace
    translationSpace=BCS_WorldSpace
    alphaBoolBlend {blendOption=Linear}
    alphaScaleBiasClamp {inRange {max=0, min=-40}, outRange {max=0, min=1}}
    boneToModify {boneName=thigh_l}
   }
   n4 = "Transform (Modify) Bone - Bone: thigh_r"(ComponentPose=@n3, Alpha=1, Translation="0,0,0", Rotation="0,0,0", Scale="1,1,1") {
    rotationMode=BMM_Replace
    rotationSpace=BCS_WorldSpace
    scaleSpace=BCS_BoneSpace
    translationMode=BMM_Replace
    translationSpace=BCS_WorldSpace
    alphaBoolBlend {blendOption=Linear}
    alphaScaleBiasClamp {inRange {max=0, min=-40}, outRange {max=0, min=1}}
    boneToModify {boneName=thigh_r}
   }
   n5 = "Transform (Modify) Bone - Bone: calf_l"(ComponentPose=@n4, Alpha=1, Translation="0,0,0", Rotation="0,0,0", Scale="1,1,1") {
    rotationMode=BMM_Replace
    rotationSpace=BCS_WorldSpace
    scaleSpace=BCS_BoneSpace
    translationMode=BMM_Replace
    translationSpace=BCS_WorldSpace
    alphaBoolBlend {blendOption=Linear}
    alphaScaleBiasClamp {inRange {max=0, min=-40}, outRange {max=0, min=1}}
    boneToModify {boneName=calf_l}
   }
   n6 = "Transform (Modify) Bone - Bone: calf_r"(ComponentPose=@n5, Alpha=1, Translation="0,0,0", Rotation="0,0,0", Scale="1,1,1") {
    rotationMode=BMM_Replace
    rotationSpace=BCS_WorldSpace
    scaleSpace=BCS_BoneSpace
    translationMode=BMM_Replace
    translationSpace=BCS_WorldSpace
    alphaBoolBlend {blendOption=Linear}
    alphaScaleBiasClamp {inRange {max=0, min=-40}, outRange {max=0, min=1}}
    boneToModify {boneName=calf_r}
   }
   n7 = "Transform (Modify) Bone - Bone: Melee_TwoHanded_Handle"!(ComponentPose=@n6, bAlphaBoolEnabled:bool=True, Translation="0,0,0", Rotation="0,0,-90", Scale="1,1,1") -> (Pose) {
    alphaInputType=Bool
    rotationMode=BMM_Additive
    rotationSpace=BCS_BoneSpace
    alphaBoolBlend {blendOption=Linear}
    boneToModify {boneName=Melee_TwoHanded_Handle}
    rotation {roll=-90}
   }
   n8 = "Transform (Modify) Bone - Bone: Melee_TwoHanded_Handle"(ComponentPose=@n7, bAlphaBoolEnabled=True, Translation="0,0,0", Rotation="0,-55,0", Scale="1,1,1") {
    alphaInputType=Bool
    rotationMode=BMM_Additive
    rotationSpace=BCS_BoneSpace
    translationMode=BMM_Additive
    alphaBoolBlend {blendOption=Linear}
    boneToModify {boneName=Melee_TwoHanded_Handle}
    rotation {yaw=-55}
   }
   n9 = "Transform (Modify) Bone - Bone: b_AR_Root"!(ComponentPose=@n8, bAlphaBoolEnabled:bool=True, Translation="15,0,15", Rotation="0,0,0", Scale="1,1,1") -> (Pose) {
    alphaInputType=Bool
    rotationMode=BMM_Additive
    rotationSpace=BCS_BoneSpace
    translationMode=BMM_Additive
    alphaBoolBlend {blendOption=Linear}
    boneToModify {boneName=b_AR_Root}
    translation {x=15, z=15}
   }
   n10 = "Transform (Modify) Bone - Bone: Melee_TwoHanded_Handle"!(ComponentPose=@n9, Alpha:float=1, Translation="0,0,0", Rotation="0,0,90", Scale="1,1,1") -> (Pose) {
    rotationMode=BMM_Additive
    rotationSpace=BCS_BoneSpace
    translationMode=BMM_Additive
    translationSpace=BCS_BoneSpace
    alphaBoolBlend {blendOption=Linear}
    boneToModify {boneName=Melee_TwoHanded_Handle}
    rotation {roll=90}
   }
   n11 = "Transform (Modify) Bone - Bone: dyn_bag_mstr"(ComponentPose=@n10, Alpha=1, Translation="0,1.5,-1.5", Rotation="0,0,90", Scale="1,1,1") {
    rotationMode=BMM_Additive
    translationMode=BMM_Additive
    alphaBoolBlend {blendOption=Linear}
    boneToModify {boneName=dyn_bag_mstr}
    rotation {roll=90}
    translation {y=1.5, z=-1.5}
   }
   n12 = "Transform (Modify) Bone - Bone: base_Handle"!(ComponentPose=@n11, bAlphaBoolEnabled:bool=True, Translation="0.2,0,-0.5", Rotation="-10,10,0", Scale="1,0.75,1.2") -> (Pose) {
    alphaInputType=Bool
    rotationMode=BMM_Additive
    scaleMode=BMM_Additive
    translationMode=BMM_Additive
    alphaBoolBlend {blendOption=Linear}
    boneToModify {boneName=base_Handle}
    rotation {pitch=-10, yaw=10}
    scale {y=0.75, z=1.2}
    translation {x=0.2, z=-0.5}
   }
   n13 = "Transform (Modify) Bone - Bone: base_Handle"(ComponentPose=@n12, bAlphaBoolEnabled=True, Translation="0,0,0", Rotation="-10,10,0", Scale="1,0.75,1.2") {
    alphaInputType=Bool
    rotationMode=BMM_Additive
    scaleMode=BMM_Additive
    translationMode=BMM_Additive
    alphaBoolBlend {blendOption=Linear}
    boneToModify {boneName=base_Handle}
    rotation {pitch=-10, yaw=10}
    scale {y=0.75, z=1.2}
   }
   n14 = "Transform (Modify) Bone - Bone: base_Handle"(ComponentPose=@n13, bAlphaBoolEnabled=True, Translation="-0.25,0,0", Rotation="-10,10,0", Scale="1,0.75,1.2") {
    alphaInputType=Bool
    rotationMode=BMM_Additive
    scaleMode=BMM_Additive
    translationMode=BMM_Additive
    alphaBoolBlend {blendOption=Linear}
    boneToModify {boneName=base_Handle}
    rotation {pitch=-10, yaw=10}
    scale {y=0.75, z=1.2}
    translation {x=-0.25}
   }
   n15 = RigidBody!(
    ComponentPose=@n14
    Alpha
    OverrideWorldGravity:Vector=Get Override World Gravity() -> (Override World Gravity:Vector)
    ExternalForce
    SimSpaceSettings:Sim Space Settings Structure=Get Sim Space Settings() -> (Sim Space Settings:Sim Space Settings Structure)
   ) -> (Pose) {
    bOverrideWorldGravity=true
    overridePhysicsAsset=&/BrineGleam_Axe_Physics
    simulationSpace=BaseBoneSpace
    alphaBoolBlend {blendOption=Linear}
    baseBoneRef {boneName=b_AR_Root}
    componentAppliedLinearAccClamp=<3500 3500 3500>
    overrideWorldGravity {z=-980}
   }
   n16 = RigidBody!(
    ComponentPose=@n15
    bAlphaBoolEnabled:bool
    OverrideWorldGravity=Get Override World Gravity()
    ExternalForce
    SimSpaceSettings=Get Sim Space Settings()
   ) -> (Pose) {
    alphaInputType=Bool
    bOverrideWorldGravity=true
    overridePhysicsAsset=&/BrineGleam_Axe_PhysicsFE_OffPawn
    simulationSpace=BaseBoneSpace
    alphaBoolBlend {blendOption=Linear}
    baseBoneRef {boneName=b_AR_Root}
    componentAppliedLinearAccClamp=<3500 3500 3500>
    overrideWorldGravity {z=-980}
   }
   n17 = RigidBody(
    ComponentPose=@n16
    OverrideWorldGravity=Get Override World Gravity()
    SimSpaceSettings=Get Sim Space Settings_FrontEnd_OnPawn() -> (Sim Space Settings_FrontEnd_OnPawn:Sim Space Settings Structure)
   ) {
    alphaInputType=Bool
    bOverrideWorldGravity=true
    overridePhysicsAsset=&/BrineGleam_Axe_PhysicsFE_OnPawn
    simulationSpace=BaseBoneSpace
    alphaBoolBlend {blendOption=Linear}
    baseBoneRef {boneName=b_AR_Root}
    componentAppliedLinearAccClamp=<3500 3500 3500>
    overrideWorldGravity {z=-980}
   }
   Output Pose(Result=Component To Local(ComponentPose=@n17)) {layerGroup=None}
  ]}
 EventGraph {nodes [
   Event BlueprintUpdateAnimation()
   > K2Node_Composite_0 = CollapseGraph!(exec:Exec) -> (output:Exec) {is=title K2Node_Composite}
   > Set RBAN off(RBAN off:bool=NOT Boolean(A=Get bRigidBodyEnabled())) -> (Output_Get:bool)
   > Set OnPlayerFE(OnPlayerFE:bool=AND Boolean(A=Get bIsInFrontEnd(), B=Get bIsOnPawn() -> (bIsOnPawn:bool))) -> (Output_Get:bool)
   n1 = Get bIsOnPawn()
   > Set OffPlayerFE(OffPlayerFE:bool=AND Boolean(A=Get bIsInFrontEnd(), B=NOT Boolean(A=@n1))) -> (Output_Get:bool)
   > n2 = Branch(Condition=Get bIsInFrontEnd())
   > Set RbanFEAlpha(RbanFEAlpha:double=1) -> (Output_Get:double)
   > Set RbanAlpha(RbanAlpha:double=0) -> (Output_Get:double)
   > Set Sim Space Settings(Sim Space Settings:Sim Space Settings Structure=Get Sim Space Settings_FrontEnd() -> (Sim Space Settings_FrontEnd:Sim Space Settings Structure)) -> (Output_Get:Sim Space Settings Structure)
   > Set Override World Gravity(Override World Gravity:Vector="0,0,-1660") -> (Output_Get:Vector)
   Set RbanFEAlpha(execute=@n2.else, RbanFEAlpha=0)
   > Set RbanAlpha(RbanAlpha=1)
   > Set Sim Space Settings(Sim Space Settings=Get Sim Space Settings_PIE() -> (Sim Space Settings_PIE:Sim Space Settings Structure))
   > Set Override World Gravity(Override World Gravity="0,0,-980")
   n3 = Event OnCharacterAndPickaxeCustomizationLoaded!() -> (
    OutputDelegate
    then
    CustomBaseMesh:Skeletal Mesh Component Object Reference
    CustomParts:Array of Custom Character Part Object References
    CustomMeshParts:Array of Skeletal Mesh Component Budgeted Object References
   ) {is=title K2Node_Event}
   > n4 = Sequence()
   > n5 = Branch(Condition=Get bIsInFrontEnd())
   > Set BrineGleam_Pickaxe_ROT_FE(BrineGleam_Pickaxe_ROT_FE:double=1) -> (Output_Get:double)
   Set BrineGleam_Pickaxe_ROT_FE(execute=@n5.else, BrineGleam_Pickaxe_ROT_FE=0)
   n6 = Get Array Item with EFort Custom Part Type(Array:Array of Custom Character Part Object References=@n3.CustomParts) -> (Output:Custom Character Part Object Reference)
   n7 = Get GenderPermitted!(self:Custom Character Part Object Reference=@n6) -> (GenderPermitted:EFortCustomGender Enum=Invalid)
   Set isMale(execute=@n4.then_1, isMale:bool="Equal (Enum)"(A:EFortCustomGender Enum=@n7, B:EFortCustomGender Enum=Male)) -> (Output_Get:bool)
   > Set isFemale(isFemale:bool="Equal (Enum)"(A=@n7, B=Female)) -> (Output_Get:bool)
   > Set isLarge(isLarge:bool="Equal (Enum)"(
     A:EFortCustomBodyType Enum=Get BodyTypesPermitted!(self:Custom Character Part Object Reference=@n6) -> (BodyTypesPermitted:EFortCustomBodyType Enum=NONE)
     B:EFortCustomBodyType Enum=Large
    )) -> (Output_Get:bool)
   > Set isMedium(isMedium:bool=NOT Boolean(A=Get isLarge() -> (isLarge:bool))) -> (Output_Get:bool)
   > Set isMaleL(isMaleL:bool=AND Boolean(A=Get isMale() -> (isMale:bool), B=Get isLarge())) -> (Output_Get:bool)
   > Set isFemaleL(isFemaleL:bool=AND Boolean(A=Get isFemale() -> (isFemale:bool), B=Get isLarge())) -> (Output_Get:bool)
  ]}
 CollapseGraph {
  path=EventGraph.K2Node_Composite_0.CollapseGraph
  nodes [
   Inputs!() -> (exec:Exec)
   n1 = GetPartSkeletalMeshComponent()
   > Is Valid(InputObject=@n1)
   n2 = GetAttachParent!(self:Scene Component Object Reference=@n1) -> (ReturnValue:Scene Component Object Reference)
   > Is Valid(InputObject=@n2)
   > Set Body_Mesh(Body_Mesh:Scene Component Object Reference=@n2) -> (Output_Get:Scene Component Object Reference)
   > n3 = Sequence!(execute) -> (then_0, then_1, then_2:Exec, then_3:Exec, then_4:Exec, then_5:Exec, then_6:Exec)
   n4 = Get Body_Mesh() -> (Body_Mesh:Scene Component Object Reference)
   n5 = BreakTransform(InTransform=GetSocketTransform(self=@n4, InSocketName=thigh_l, TransformSpace=RTS_World))
   Set thigh_left_location(execute=@n3.then_1, thigh_left_location:Vector=@n5) -> (Output_Get:Vector)
   > Set thigh_left_rotation(thigh_left_rotation:Rotator=@n5.Rotation) -> (Output_Get:Rotator)
   n6 = BreakTransform(InTransform=GetSocketTransform(self=@n4, InSocketName=root, TransformSpace=RTS_World))
   Set root_location(execute=@n3.then_6, root_location:Vector=@n6) -> (Output_Get:Vector)
   > Set root_space_rotation(root_space_rotation:Rotator=@n6.Rotation) -> (Output_Get:Rotator)
   > Outputs!(output:Exec)
   n7 = BreakTransform(InTransform=GetSocketTransform(self=@n4, InSocketName=pelvis, TransformSpace=RTS_World))
   Set pelvis_location(execute=@n3.then_0, pelvis_location:Vector=@n7) -> (Output_Get:Vector)
   > Set pelvis_rotation(pelvis_rotation:Rotator=@n7.Rotation) -> (Output_Get:Rotator)
   n8 = BreakTransform(InTransform=GetSocketTransform(self=@n4, InSocketName=thigh_r, TransformSpace=RTS_World))
   Set thigh_right_location(execute=@n3.then_2, thigh_right_location:Vector=@n8) -> (Output_Get:Vector)
   > Set thigh_right_rotation(thigh_right_rotation:Rotator=@n8.Rotation) -> (Output_Get:Rotator)
   n9 = BreakTransform(InTransform=GetSocketTransform(self=@n4, InSocketName=calf_r, TransformSpace=RTS_World))
   Set calf_right_location(execute=@n3.then_4, calf_right_location:Vector=@n9) -> (Output_Get:Vector)
   > Set calf_right_rotation(calf_right_rotation:Rotator=@n9.Rotation) -> (Output_Get:Rotator)
   n10 = BreakTransform(InTransform=GetSocketTransform(self=@n4, InSocketName=calf_l, TransformSpace=RTS_World))
   Set calf_left_location(execute=@n3.then_3, calf_left_location:Vector=@n10) -> (Output_Get:Vector)
   > Set calf_left_rotation(calf_left_rotation:Rotator=@n10.Rotation) -> (Output_Get:Rotator)
   n11 = BreakTransform(InTransform=GetSocketTransform(self=@n4, InSocketName=root, TransformSpace=RTS_World))
   Set body_root_location(execute=@n3.then_5, body_root_location:Vector=@n11) -> (Output_Get:Vector)
   > Set body_root_rotation(body_root_rotation:Rotator=@n11.Rotation) -> (Output_Get:Rotator)
  ]
 }
]
variables [
 Body_Mesh {type=SceneComponent, default=null}
 body_root_location {type=Vector}
 body_root_rotation {type=Rotator}
 BrineGleam_Pickaxe_ROT_FE {type=number}
 calf_left_location {type=Vector}
 calf_left_rotation {type=Rotator}
 calf_right_location {type=Vector}
 calf_right_rotation {type=Rotator}
 isFemale {type=boolean}
 isFemaleL {type=boolean}
 isLarge {type=boolean}
 isMale {type=boolean}
 isMaleL {type=boolean}
 isMedium {type=boolean}
 OffPlayerFE {type=boolean}
 OnPlayerFE {type=boolean}
 Override World Gravity {type=Vector}
 pelvis_location {type=Vector}
 pelvis_rotation {type=Rotator}
 RBAN off {type=boolean}
 RbanAlpha {type=number}
 RbanFEAlpha {type=number}
 root_location {type=Vector}
 root_space_rotation {type=Rotator}
 Sim Space Settings {
  type=SimSpaceSettings
  default {
   dampingAlpha=1
   maxAngularAcceleration=10000
   maxAngularVelocity=10000
   maxLinearAcceleration=10000
   maxLinearVelocity=10000
   velocityScaleZ=1
   worldAlpha=0
   externalLinearDragV=<0 0 0>
  }
 }
 Sim Space Settings_FrontEnd {
  type=SimSpaceSettings
  default {
   dampingAlpha=0
   maxAngularAcceleration=750
   maxAngularVelocity=150
   maxLinearAcceleration=1500
   maxLinearVelocity=1000
   velocityScaleZ=0.25
   worldAlpha=1
   externalLinearDragV=<0.2 0.2 0.2>
  }
 }
 Sim Space Settings_FrontEnd_OnPawn {
  type=SimSpaceSettings
  default {
   dampingAlpha=0
   maxAngularAcceleration=6750
   maxAngularVelocity=450
   maxLinearAcceleration=4500
   maxLinearVelocity=3000
   velocityScaleZ=0.75
   worldAlpha=1
   externalAngularVelocity=<0.5 0.5 0.5>
   externalLinearDragV=<0.5 0.5 0.5>
   externalLinearVelocity=<0.5 0.5 0.5>
  }
 }
 Sim Space Settings_PIE {
  type=SimSpaceSettings
  default {
   dampingAlpha=0
   maxAngularAcceleration=1000
   maxAngularVelocity=250
   maxLinearAcceleration=2000
   maxLinearVelocity=1500
   velocityScaleZ=0.5
   worldAlpha=1
   externalLinearDragV=<0.3 0.3 0.3>
  }
 }
 Skel_Mesh {type=SkeletalMeshComponent}
 thigh_left_location {type=Vector}
 thigh_left_rotation {type=Rotator}
 thigh_right_location {type=Vector}
 thigh_right_rotation {type=Rotator}
]
